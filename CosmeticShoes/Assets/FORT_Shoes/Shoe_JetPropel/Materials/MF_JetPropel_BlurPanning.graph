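# Material function for the JetPropel cosmetic shoe that fakes a spinning motor with panning, pre-blurred stripe textures.
# Time and Spin Speed scroll UVs along V, while scaled, saturated U confines the stripes to a band.
# The stripe mask darkens base color, a flattened stripe normal blends into the incoming one, and U-based masks drive orange-red emissive.
# Red vertex color blends the result with another material-attributes input, limiting where the effect appears.
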MaterialFunction /CosmeticShoes/Assets/FORT_Shoes/Shoe_JetPropel/Materials/MF_JetPropel_BlurPanning
graph {nodes [
  n1 = NamedRerouteUsage()
  n2 = Constant3Vector() {constant=<rgba 0.5 0.5 0.5 1>}
  n3 = Add(A=TextureCoordinate(), B=AppendVector(A=Constant() {r=-0.957}, B=Constant()))
  n4 = Multiply(A=Time(), B=AppendVector(A=Constant(), B=ScalarParameter() {defaultValue=10, parameterName=Spin Speed}))
  n5 = Constant() {r=45}
  n6 = Constant() {r=20}
  n7 = MaterialFunctionCall(Float2=Multiply(A=Add(A=@n3, B=@n4), B=AppendVector(A=@n5, B=@n6))) {materialFunction=&BreakOutFloat2Components}
  n8 = AppendVector(A=Saturate(None=@n7.R), B=@n7.G)
  n9 = TextureSample(UVs=@n8) {
   samplerSource=SSM_Wrap_WorldGroupSettings
   samplerType=SAMPLERTYPE_Masks
   texture=/CosmeticShoes/Assets/FORT_Shoes/Shoe_JetPropel/Textures/T_JetPropel_MotorStripes_Blurred_basecolor
  }
  n10 = TextureSample(UVs=@n8) {
   samplerSource=SSM_Wrap_WorldGroupSettings
   samplerType=SAMPLERTYPE_Normal
   texture=/CosmeticShoes/Assets/FORT_Shoes/Shoe_JetPropel/Textures/T_JetPropel_MotorStripes_Blurred_normal
  }
  n11 = OneMinus(None=Constant() {r=0.5})
  n12 = Step(Y=ComponentMask(None=TextureCoordinate()) {r=true}, X=Constant() {r=0.5})
  n13 = ComponentMask(None=TextureCoordinate()) {r=true}
  n14 = Power(Base=Saturate(None=Add(A=@n13) {constB=0.01})) {constExponent=70}
  n15 = Constant3Vector() {constant=<rgba 2 0.1667 0 0>}
  n16 = NamedRerouteUsage()
  n17 = SetMaterialAttributes!(
   MaterialAttributes=NamedRerouteUsage()
   BaseColor=LinearInterpolate(A=@n1, B=Multiply(A=@n1, B=@n2), Alpha=NamedRerouteDeclaration(None=@n9.R) {name=SpinMask, nodeColor=<rgba 0.6172 0 1 1>})
   Normal=MaterialFunctionCall(
    BaseNormal=NamedRerouteUsage()
    AdditionalNormal=MaterialFunctionCall(Normal=@n10.RGB, Flatness=@n11) {materialFunction=&FlattenNormal}.Result
   ) {materialFunction=&BlendAngleCorrectedNormals}.Result
   EmissiveColor=Multiply(A=Add(A=@n12, B=@n14), B=Subtract(A=@n15, B=@n16))
  ) {attributeSetTypes [#1, #2, #3]}
  FunctionOutput(None=BlendMaterialAttributes(A=NamedRerouteUsage(), B=NamedRerouteDeclaration(None=@n17) {name=OutMat, nodeColor=<rgba 1 0 0.2578 1>}, Alpha=VertexColor().R))
  n18 = GetMaterialAttributes(None=NamedRerouteDeclaration(None=FunctionInput() {inputType=FunctionInput_MaterialAttributes}) {name=InMat, nodeColor=<rgba 1 0.7031 0 1>}) {attributeGetTypes [#1, #2]}
  NamedRerouteDeclaration(None=@n18.BaseColor) {name=InBC, nodeColor=<rgba 0.2188 0 1 1>}
  NamedRerouteDeclaration(None=@n18.Normal) {name=InN, nodeColor=<rgba 0.1016 1 0 1>}
 ]}
