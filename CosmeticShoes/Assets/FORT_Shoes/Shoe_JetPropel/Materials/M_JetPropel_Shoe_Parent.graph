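# Parent material for the JetPropel cosmetic shoe, built as a chain of shared material functions passing material attributes.
# The 2023 character-parent PreFX output runs through the JetPropel blur-panning function, shoe reactivity and sock functions, with the Tech Art Mask driving the sock mask.
# PostFX and a tech-art shoe hologram function finish the chain.
# Shading model comes from the graph, and pixel animation is flagged.
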Material /CosmeticShoes/Assets/FORT_Shoes/Shoe_JetPropel/Materials/M_JetPropel_Shoe_Parent
graph {
 properties {
  bHasPixelAnimation=true
  bUseMaterialAttributes=true
  bUsedWithClothing=true
  bUsedWithMeshDeformer=true
  bUsedWithMorphTargets=true
  bUsedWithNanite=true
  bUsedWithNiagaraMeshParticles=true
  bUsedWithSkeletalMesh=true
  pixelDepthOffsetMode=PDOM_Legacy
  shadingModel=MSM_FromMaterialExpression
  usedShadingModels=Default Lit | Subsurface | Clear Coat | Subsurface Profile | Cloth
 }
 nodes [
  n1 = MaterialFunctionCall() {materialFunction=&MF_CharacterParent_2023_PreFX}
  n2 = MaterialFunctionCall!(Material=MaterialFunctionCall!(In=@n1.Result) {materialFunction=&/MF_JetPropel_BlurPanning}.Result) {materialFunction=/Game/Global/Materials/Cosmetics/Kicks/Shoes/Materials/MF_ShoeReactivityTest}
  MaterialFunctionCall!(In=MaterialFunctionCall(
    Material=MaterialFunctionCall!(InOpacity, Material=@n2.Result, Mask=@n1.Tech Art Mask) {materialFunction=/Game/Global/Materials/Cosmetics/Kicks/Socks/Materials/MF_Sock}.Result
    M=@n1.M
    Pre Skinned Local Position=@n1.M
    Distance Blend=@n1.M
    Fresnel=@n1.M
   ) {materialFunction=&MF_CharacterParent_2023_PostFX}.Result) {materialFunction=/Game/Global/Materials/Cosmetics/Kicks/Shoes/Materials/MF_TechartShoeHologram}
 ]
}
